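; Parent Blueprint for the Commerce clothing rack dress prop in Fortnite, which tints the Clothes mesh on the rack.
; When more than one static mesh component exists, the construction script passes a color and the Clothes mesh to SetColoronSecondaryMesh.
; That color is the last building texture data entry's tint when there are multiple entries, otherwise the ClothingColor variable.
; The event graph holds only empty event stubs.
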(event EventBeginPlay
  (|CallParentFunction) ; This node is disabled and will not be called.\nDrag off pins to build functionality.
)

(event Collision|EventActorBeginOverlap
  (params (OtherActor :type Actor%20Object%20Reference)))

(event EventTick
  (params (DeltaSeconds :type Float%20%28single-precision%29)))

(fn ConstructionScript
  (bind _clothes (Variables|Default|GetClothes))
  (|CallParentFunction)
  (if (> (Utilities|Array|Length (Actor|GetComponentsByClass "/Script/Engine.StaticMeshComponent")) 1)
    (bind _outtexturedataarray (Building|GetBuildingTextureData))
    (if (> (Utilities|Array|Length _outtexturedataarray) 1)
      (Default|SetColoronSecondaryMesh (Math|Conversions|ToLinearColor(Color) (Variables|/Script/FortniteGame.BuildingTextureData|GetTintColor (Utilities|Array|Get(acopy) _outtexturedataarray (- (Utilities|Array|Length _outtexturedataarray) 1)))) _clothes)
      (else
        (Default|SetColoronSecondaryMesh (Variables|Default|GetClothingColor) _clothes)))))
